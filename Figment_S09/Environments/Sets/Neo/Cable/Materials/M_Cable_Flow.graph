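# Lit material for Neo set cables, usable on spline meshes, that animates flowing two-color electric glow along the cable.
# A shared panner scrolls a grayscale Time texture, sampled at offsets to blend colors and shape pulses.
# Multiplied noise samples add electric flicker to the diffuse glow, while spec alpha masks it and, raised to a power, the core glow.
# The pulse also builds a vertical normal-based displacement, weighted by vertex alpha.
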Material /Figment_S09/Environments/Sets/Neo/Cable/Materials/M_Cable_Flow
graph {
 properties {
  bUsedWithSplineMeshes=true
  pixelDepthOffsetMode=PDOM_Legacy
  previewMesh=/Figment_Sniper_Weather_Athena/Gameplay/S_SniperScope_FigmentWeatherDisplay
  usedShadingModels=Default Lit
 }
 nodes [
  n1 = TextureSampleParameter2D() {parameterName=Diffuse, texture=/Figment_S09/Environments/Sets/Neo/Cable/Textures/T_NeoT_Cable_D}
  n2 = TextureSample(UVs=Multiply(A=TextureCoordinate() {coordinateIndex=1}, B=ScalarParameter() {defaultValue=2, parameterName=Param_1})) {samplerType=SAMPLERTYPE_LinearColor, texture=/Game/Characters/Player/Male/Medium/Bodies/M_MED_SRDrift/Textures/T_ElectricNoise}
  n3 = TextureSample() {samplerType=SAMPLERTYPE_LinearColor, texture=&T_ElectricNoise}
  n4 = Panner(
   Coordinate=TextureCoordinate() {coordinateIndex=1, vTiling=0.35}
   Speed=VectorParameter() {parameterName=Speed, defaultValue=<rgba -1.5 0 0 0>}.RGB
  )
  n5 = TextureObject() {samplerType=SAMPLERTYPE_LinearGrayscale, texture=/Game/Environments/Sets/Neo/Cable/Textures/Time}
  n6 = TextureSample(UVs=@n4, Tex=@n5) {samplerType=SAMPLERTYPE_LinearGrayscale, texture=&Time}
  n7 = LinearInterpolate(
   A=VectorParameter() {parameterName=C2, defaultValue=<rgba 0.7396 0.5045 5 0>}.RGB
   B=VectorParameter() {parameterName=C1, defaultValue=<rgba 0 0.8996 5 0>}.RGB
   Alpha=@n6.B
  )
  n8 = ComponentMask(None=Constant3Vector() {constant=<rgba -0.2952 0.6 0 0>}) {g=true, r=true}
  n9 = Multiply(
   A=@n7
   B=TextureSample(UVs=Add(A=@n4, B=@n8), Tex=@n5) {samplerType=SAMPLERTYPE_LinearGrayscale, texture=&Time}.R
  )
  n10 = ScalarParameter() {defaultValue=300, parameterName=ElectricGlow}
  n11 = Add(
   A=Multiply(A=Multiply(A=@n2.R, B=@n3.R), B=Multiply(A=@n9, B=@n10))
   B=Multiply(A=@n9, B=ScalarParameter() {defaultValue=10, parameterName=CableGlow})
  )
  n12 = TextureSampleParameter2D() {parameterName=Spec, samplerType=SAMPLERTYPE_Masks, texture=/Figment_S09/Environments/Sets/Neo/Cable/Textures/T_NeoT_Cable_S}
  Add(A=@n1.RGB, B=Multiply(A=Multiply(A=@n1.RGB, B=@n11), B=@n12.A))
  n13 = LinearInterpolate(A=Constant() {r=0.02}, Alpha=@n6.B)
  n14 = ComponentMask(None=Constant3Vector() {constant=<rgba -0.2 0.6 0 0>}) {g=true, r=true}
  n15 = Multiply(
   A=@n7
   B=TextureSample(UVs=Add(A=@n4, B=@n14), Tex=@n5) {samplerType=SAMPLERTYPE_LinearGrayscale, texture=&Time}.R
  )
  n16 = ScalarParameter() {defaultValue=5, parameterName=CoreGlow}
  Multiply(A=Power(Base=@n12.A, Exp=Constant() {r=5}), B=Add(A=Multiply(A=@n7, B=@n13), B=Multiply(A=@n15, B=@n16)))
  n17 = ScalarParameter() {defaultValue=3, parameterName=Displace}
  n18 = VertexNormalWS()
  n19 = Constant3Vector() {constant=<rgba 0 0 1 1>}
  Multiply(A=Multiply(A=Multiply(A=@n6.B, B=@n17), B=Multiply(A=@n18, B=@n19)), B=VertexColor().A)
  TextureSampleParameter2D() {parameterName=Normal, samplerType=SAMPLERTYPE_Normal, texture=/Figment_S09/Environments/Sets/Neo/Cable/Textures/T_NeoT_Cable_N}
 ]
}
